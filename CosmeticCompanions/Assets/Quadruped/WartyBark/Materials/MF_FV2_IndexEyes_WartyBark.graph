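# Material function for the WartyBark quadruped companion that shows flipbook eye and brow shapes by hiding unselected vertex-colored geometry.
# Left and right index parameters, chosen by pre-skinning local X, go through curve atlas remaps, with a switch substituting one eye index.
# Sphere masks match brow indices to red vertex color and eye indices to green.
# Unmatched non-white vertices collapse to the object origin through World Position Offset.
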MaterialFunction /CosmeticCompanions/Assets/Quadruped/WartyBark/Materials/MF_FV2_IndexEyes_WartyBark
graph {expressions {
  MaterialExpressionAdd_0 {
   class=MaterialExpressionAdd
   inputs [
    A {from=MaterialExpressionSphereMask_2}
    B {from=MaterialExpressionSphereMask_1}
   ]
  }
  MaterialExpressionAdd_3 {
   class=MaterialExpressionAdd
   inputs [
    A {from=MaterialExpressionGetMaterialAttributes_1, output=WorldPositionOffset}
    B {from=MaterialExpressionMultiply_2}
   ]
  }
  MaterialExpressionConstant3Vector_1 {class=MaterialExpressionConstant3Vector, properties {constant=<rgba 1 1 1 0>}}
  MaterialExpressionConstant_1 {class=MaterialExpressionConstant, properties {r=6}}
  MaterialExpressionConstant_10 {class=MaterialExpressionConstant, properties {r=3}}
  MaterialExpressionConstant_11 {class=MaterialExpressionConstant}
  MaterialExpressionConstant_9 {class=MaterialExpressionConstant, properties {r=1}}
  MaterialExpressionCurveAtlasRowParameter_1 {
   class=MaterialExpressionCurveAtlasRowParameter
   inputs [CurveTime {from=MaterialExpressionDivide_3, output=""}]
   properties {
    atlas=/Game/Characters/Player/Common/MaterialFunctions/VertexColorFaces/CA_FlipbookV2_Remapping
    curve=/Game/Characters/Player/Common/MaterialFunctions/VertexColorFaces/C_EyesAndMouthDefault
    group=Flipbook Eyes
    parameterName=BrowIndexRemapCurve
   }
  }
  MaterialExpressionCurveAtlasRowParameter_2 {
   class=MaterialExpressionCurveAtlasRowParameter
   inputs [CurveTime {from=MaterialExpressionDivide_4, output=""}]
   properties {
    atlas=&CA_FlipbookV2_Remapping
    curve=&C_EyesAndMouthDefault
    group=Flipbook Eyes
    parameterName=EyeIndexRemapCurve
   }
  }
  MaterialExpressionDivide_3 {
   class=MaterialExpressionDivide
   inputs [
    A {from=MaterialExpressionLinearInterpolate_1}
    B {from=""}
   ]
   properties {constB=10}
  }
  MaterialExpressionDivide_4 {
   class=MaterialExpressionDivide
   inputs [
    A {from=MaterialExpressionSwitch_1}
    B {from=""}
   ]
   properties {constB=10}
  }
  MaterialExpressionDotProduct_1 {
   class=MaterialExpressionDotProduct
   inputs [
    A {from=MaterialExpressionConstant3Vector_1, output=""}
    B {from=MaterialExpressionVertexColor_4, output=""}
   ]
   properties {}
  }
  MaterialExpressionFunctionInput_1 {
   class=MaterialExpressionFunctionInput
   inputs [Preview {from="", output=""}]
   properties {inputName=Material, inputType=FunctionInput_MaterialAttributes}
  }
  MaterialExpressionFunctionOutput_1 {class=MaterialExpressionFunctionOutput, inputs [None {from=MaterialExpressionSetMaterialAttributes_1, output=""}]}
  MaterialExpressionGetMaterialAttributes_1 {
   class=MaterialExpressionGetMaterialAttributes
   inputs [None {from=MaterialExpressionFunctionInput_1, output=""}]
   properties {attributeGetTypes [#1]}
  }
  MaterialExpressionIf_1 {
   class=MaterialExpressionIf
   inputs [
    A {from=MaterialExpressionDotProduct_1, output=""}
    B {from=MaterialExpressionConstant_10, output=""}
    "A > B" {from=MaterialExpressionConstant_11, output=""}
    "A == B" {from=MaterialExpressionConstant_11, output=""}
    "A < B" {from=MaterialExpressionConstant_9, output=""}
   ]
   properties {}
  }
  MaterialExpressionLinearInterpolate_0 {
   class=MaterialExpressionLinearInterpolate
   inputs [
    A {from=MaterialExpressionScalarParameter_6}
    B {from=MaterialExpressionScalarParameter_7}
    Alpha {from=MaterialExpressionNamedRerouteUsage_2}
   ]
  }
  MaterialExpressionLinearInterpolate_1 {
   class=MaterialExpressionLinearInterpolate
   inputs [
    A {from=MaterialExpressionScalarParameter_1}
    B {from=MaterialExpressionScalarParameter_3}
    Alpha {from=MaterialExpressionNamedRerouteUsage_3}
   ]
  }
  MaterialExpressionLocalPosition_1 {class=MaterialExpressionLocalPosition, inputs [], properties {localOrigin=InstancePreSkinning}}
  MaterialExpressionMaterialFunctionCall_1 {
   class=MaterialExpressionMaterialFunctionCall
   inputs [Float3 {from=MaterialExpressionLocalPosition_1, output=XYZ}]
   properties {materialFunction=&BreakOutFloat3Components}
  }
  MaterialExpressionMultiply_0 {
   class=MaterialExpressionMultiply
   inputs [
    A {from=MaterialExpressionSubtract_0}
    B {from=""}
   ]
   properties {constB=-1}
  }
  MaterialExpressionMultiply_2 {
   class=MaterialExpressionMultiply
   inputs [
    A {from=MaterialExpressionNamedRerouteUsage_1}
    B {from=MaterialExpressionSaturate_1}
   ]
  }
  MaterialExpressionNamedRerouteDeclaration_1 {
   class=MaterialExpressionNamedRerouteDeclaration
   inputs [None {from=MaterialExpressionMultiply_0}]
   properties {name=Vertex Color Hide, nodeColor=<rgba 0.6875 0 1 1>}
  }
  MaterialExpressionNamedRerouteDeclaration_2 {
   class=MaterialExpressionNamedRerouteDeclaration
   inputs [None {from=MaterialExpressionOneMinus_1}]
   properties {name=Left0/Right1, nodeColor=<rgba 1 0 0.1172 1>}
  }
  MaterialExpressionNamedRerouteUsage_1 {class=MaterialExpressionNamedRerouteUsage}
  MaterialExpressionNamedRerouteUsage_2 {class=MaterialExpressionNamedRerouteUsage}
  MaterialExpressionNamedRerouteUsage_3 {class=MaterialExpressionNamedRerouteUsage}
  MaterialExpressionObjectPositionWS_1 {class=MaterialExpressionObjectPositionWS, inputs [], properties {}}
  MaterialExpressionOneMinus_1 {class=MaterialExpressionOneMinus, inputs [None {from=MaterialExpressionSaturate_3}]}
  MaterialExpressionRound_1 {class=MaterialExpressionRound, inputs [None {from=MaterialExpressionSubtract_2, output=""}], properties {}}
  MaterialExpressionRound_3 {class=MaterialExpressionRound, inputs [None {from=MaterialExpressionMaterialFunctionCall_1, output=R}], properties {}}
  MaterialExpressionSaturate_1 {class=MaterialExpressionSaturate, inputs [None {from=MaterialExpressionRound_1}]}
  MaterialExpressionSaturate_3 {class=MaterialExpressionSaturate, inputs [None {from=MaterialExpressionRound_3}]}
  MaterialExpressionScalarParameter_1 {class=MaterialExpressionScalarParameter, properties {group=Flipbook Eyes, parameterName=Flipbook_face_L_brow_index}}
  MaterialExpressionScalarParameter_3 {class=MaterialExpressionScalarParameter, properties {group=Flipbook Eyes, parameterName=Flipbook_face_R_brow_index}}
  MaterialExpressionScalarParameter_6 {class=MaterialExpressionScalarParameter, properties {group=Flipbook Eyes, parameterName=flipbook_face_L_eye_index}}
  MaterialExpressionScalarParameter_7 {class=MaterialExpressionScalarParameter, properties {group=Flipbook Eyes, parameterName=flipbook_face_R_eye_index}}
  MaterialExpressionSetMaterialAttributes_1 {
   class=MaterialExpressionSetMaterialAttributes
   inputs [
    MaterialAttributes {from=MaterialExpressionGetMaterialAttributes_1, output=MaterialAttributes}
    WorldPositionOffset {from=MaterialExpressionAdd_3, output=""}
   ]
   properties {attributeSetTypes [#1]}
  }
  MaterialExpressionSphereMask_1 {
   class=MaterialExpressionSphereMask
   inputs [
    A {from=MaterialExpressionCurveAtlasRowParameter_1, output=R}
    B {from=MaterialExpressionVertexColor_2, output=R}
    Radius {from="", output=""}
    Hardness {from="", output=""}
   ]
   properties {attenuationRadius=0.02}
  }
  MaterialExpressionSphereMask_2 {
   class=MaterialExpressionSphereMask
   inputs [
    A {from=MaterialExpressionCurveAtlasRowParameter_2, output=R}
    B {from=MaterialExpressionVertexColor_5, output=G}
    Radius {from="", output=""}
    Hardness {from="", output=""}
   ]
   properties {attenuationRadius=0.02}
  }
  MaterialExpressionSubtract_0 {
   class=MaterialExpressionSubtract
   inputs [
    A {from=MaterialExpressionWorldPosition_1, output=XYZ}
    B {from=MaterialExpressionObjectPositionWS_1}
   ]
  }
  MaterialExpressionSubtract_2 {
   class=MaterialExpressionSubtract
   inputs [
    A {from=MaterialExpressionIf_1}
    B {from=MaterialExpressionAdd_0}
   ]
  }
  MaterialExpressionSwitch_1 {
   class=MaterialExpressionSwitch
   inputs [
    SwitchValue {from=MaterialExpressionLinearInterpolate_0, output=""}
    Default {from=MaterialExpressionLinearInterpolate_0, output=""}
    None {from=MaterialExpressionLinearInterpolate_0, output=""}
    None {from=MaterialExpressionLinearInterpolate_0, output=""}
    None {from=MaterialExpressionConstant_1, output=""}
    None {from=MaterialExpressionLinearInterpolate_0, output=""}
    None {from=MaterialExpressionLinearInterpolate_0, output=""}
    None {from=MaterialExpressionLinearInterpolate_0, output=""}
    None {from=MaterialExpressionLinearInterpolate_0, output=""}
    None {from=MaterialExpressionLinearInterpolate_0, output=""}
   ]
   properties {
    constDefault=6
    inputs [
     {
      inputName=None
      input {
       expression=&self:MaterialExpressionLinearInterpolate_0
       inputName=None
       mask=0
       maskA=0
       maskB=0
       maskG=0
       maskR=0
       outputIndex=0
      }
     }
     {
      inputName=None
      input {
       expression=&self:MaterialExpressionLinearInterpolate_0
       inputName=None
       mask=0
       maskA=0
       maskB=0
       maskG=0
       maskR=0
       outputIndex=0
      }
     }
     {
      inputName=None
      input {
       expression=&self:MaterialExpressionConstant_1
       inputName=None
       mask=0
       maskA=0
       maskB=0
       maskG=0
       maskR=0
       outputIndex=0
      }
     }
     {
      inputName=None
      input {
       expression=&self:MaterialExpressionLinearInterpolate_0
       inputName=None
       mask=0
       maskA=0
       maskB=0
       maskG=0
       maskR=0
       outputIndex=0
      }
     }
     {
      inputName=None
      input {
       expression=&self:MaterialExpressionLinearInterpolate_0
       inputName=None
       mask=0
       maskA=0
       maskB=0
       maskG=0
       maskR=0
       outputIndex=0
      }
     }
     {
      inputName=None
      input {
       expression=&self:MaterialExpressionLinearInterpolate_0
       inputName=None
       mask=0
       maskA=0
       maskB=0
       maskG=0
       maskR=0
       outputIndex=0
      }
     }
     {
      inputName=None
      input {
       expression=&self:MaterialExpressionLinearInterpolate_0
       inputName=None
       mask=0
       maskA=0
       maskB=0
       maskG=0
       maskR=0
       outputIndex=0
      }
     }
     {
      inputName=None
      input {
       expression=&self:MaterialExpressionLinearInterpolate_0
       inputName=None
       mask=0
       maskA=0
       maskB=0
       maskG=0
       maskR=0
       outputIndex=0
      }
     }
    ]
   }
  }
  MaterialExpressionVertexColor_2 {class=MaterialExpressionVertexColor, inputs [], properties {}}
  MaterialExpressionVertexColor_4 {class=MaterialExpressionVertexColor, inputs [], properties {}}
  MaterialExpressionVertexColor_5 {class=MaterialExpressionVertexColor, inputs [], properties {}}
  MaterialExpressionWorldPosition_1 {class=MaterialExpressionWorldPosition, inputs [], properties {}}
 }}
