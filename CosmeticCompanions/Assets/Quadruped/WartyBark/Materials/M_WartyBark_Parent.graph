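# Parent material for the WartyBark quadruped cosmetic companion, wrapping the shared 2023 character PreFX and PostFX functions with a Subsurface Profile body whose opacity is adjusted by the Tech Art Mask.
# IsHead adds the companion flipbook and, via WartyBrine_MaskOn, throws non-white vertices away with huge World Position Offset, exempting blue-channel matches selected by mouth index.
# IsFaceAcc instead offsets PreFX vertices by red-channel sphere mask, targeted by WartyBrine_UseEyeMask.
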Material /CosmeticCompanions/Assets/Quadruped/WartyBark/Materials/M_WartyBark_Parent
graph {
 expressions {
  MaterialExpressionClearCoatNormalCustomOutput_0 {
   class=MaterialExpressionClearCoatNormalCustomOutput
   inputs [None {from=MaterialExpressionGetMaterialAttributes_1, output=ClearCoatBottomNormal}]
   properties {}
  }
  MaterialExpressionComponentMask_0 {
   class=MaterialExpressionComponentMask
   inputs [None {from=MaterialExpressionMaterialFunctionCall_0, output=Tech Art Mask}]
   properties {b=true}
  }
  MaterialExpressionConstant2Vector_0 {class=MaterialExpressionConstant2Vector, properties {g=0.2, r=0.3}}
  MaterialExpressionConstant3Vector_1 {class=MaterialExpressionConstant3Vector, properties {constant=<rgba 1 1 1 0>}}
  MaterialExpressionConstant_0 {class=MaterialExpressionConstant, properties {r=0.1}}
  MaterialExpressionConstant_1 {class=MaterialExpressionConstant}
  MaterialExpressionConstant_10 {class=MaterialExpressionConstant, properties {r=3}}
  MaterialExpressionConstant_11 {class=MaterialExpressionConstant}
  MaterialExpressionConstant_9 {class=MaterialExpressionConstant, properties {r=10000000000}}
  MaterialExpressionDotProduct_1 {
   class=MaterialExpressionDotProduct
   inputs [
    A {from=MaterialExpressionConstant3Vector_1, output=""}
    B {from=MaterialExpressionVertexColor_4, output=""}
   ]
   properties {}
  }
  MaterialExpressionGetMaterialAttributes_0 {
   class=MaterialExpressionGetMaterialAttributes
   inputs [None {from=MaterialExpressionMaterialFunctionCall_0, output=Result}]
   properties {attributeGetTypes [#1]}
  }
  MaterialExpressionGetMaterialAttributes_1 {
   class=MaterialExpressionGetMaterialAttributes
   inputs [None {from=MaterialExpressionMaterialFunctionCall_1, output=Result}]
   properties {attributeGetTypes [#2]}
  }
  MaterialExpressionGetMaterialAttributes_2 {
   class=MaterialExpressionGetMaterialAttributes
   inputs [None {from=MaterialExpressionMaterialFunctionCall_2, output=Result}]
   properties {attributeGetTypes [#3, #4]}
  }
  MaterialExpressionIf_1 {
   class=MaterialExpressionIf
   inputs [
    A {from=MaterialExpressionDotProduct_1, output=""}
    B {from=MaterialExpressionConstant_10, output=""}
    "A > B" {from=MaterialExpressionConstant_11, output=""}
    "A == B" {from=MaterialExpressionConstant_11, output=""}
    "A < B" {from=MaterialExpressionConstant_9, output=""}
   ]
   properties {}
  }
  MaterialExpressionLinearInterpolate_0 {
   class=MaterialExpressionLinearInterpolate
   inputs [
    A {from=MaterialExpressionGetMaterialAttributes_0, output=Opacity}
    B {from=MaterialExpressionMultiply_0}
    Alpha {from=MaterialExpressionComponentMask_0}
   ]
  }
  MaterialExpressionLinearInterpolate_1 {
   class=MaterialExpressionLinearInterpolate
   inputs [
    A {from=""}
    B {from=""}
    Alpha {from=MaterialExpressionScalarParameter_4}
   ]
   properties {constA=1, constB=0.1}
  }
  MaterialExpressionLinearInterpolate_2 {
   class=MaterialExpressionLinearInterpolate
   inputs [
    A {from=MaterialExpressionLinearInterpolate_3}
    B {from=""}
    Alpha {from=MaterialExpressionSphereMask_0}
   ]
   properties {constB=0}
  }
  MaterialExpressionLinearInterpolate_3 {
   class=MaterialExpressionLinearInterpolate
   inputs [
    A {from=MaterialExpressionGetMaterialAttributes_2, output=World Position Offset}
    B {from=MaterialExpressionIf_1}
    Alpha {from=MaterialExpressionScalarParameter_3}
   ]
  }
  MaterialExpressionMaterialFunctionCall_0 {
   class=MaterialExpressionMaterialFunctionCall
   inputs [
    Overide Texture Set UVs {from=""}
    Override Tech Art Mask UVs {from=""}
    Override Basic Emissive UVs {from=""}
   ]
   properties {materialFunction=&MF_CharacterParent_2023_PreFX}
  }
  MaterialExpressionMaterialFunctionCall_1 {
   class=MaterialExpressionMaterialFunctionCall
   inputs [
    Material {from=MaterialExpressionStaticSwitchParameter_1}
    M {from=MaterialExpressionMaterialFunctionCall_0, output=M}
    Pre Skinned Local Position {from=MaterialExpressionMaterialFunctionCall_0, output=M}
    Distance Blend {from=MaterialExpressionMaterialFunctionCall_0, output=M}
    Fresnel {from=MaterialExpressionMaterialFunctionCall_0, output=M}
   ]
   properties {materialFunction=&MF_CharacterParent_2023_PostFX}
  }
  MaterialExpressionMaterialFunctionCall_2 {
   class=MaterialExpressionMaterialFunctionCall
   inputs [Material {from=MaterialExpressionSetMaterialAttributes_0}]
   properties {materialFunction=/CosmeticCompanions/Materials/MaterialFunctions/MF_FN_Companion_Flipbook}
  }
  MaterialExpressionMax_0 {
   class=MaterialExpressionMax
   inputs [
    A {from=MaterialExpressionSwitch_0}
    B {from=""}
   ]
   properties {constB=0.05}
  }
  MaterialExpressionMultiply_0 {
   class=MaterialExpressionMultiply
   inputs [
    A {from=MaterialExpressionScalarParameter_1}
    B {from=MaterialExpressionComponentMask_0}
   ]
  }
  MaterialExpressionMultiply_2 {
   class=MaterialExpressionMultiply
   inputs [
    A {from=MaterialExpressionSphereMask_1}
    B {from=""}
   ]
   properties {constB=100000}
  }
  MaterialExpressionScalarParameter_1 {class=MaterialExpressionScalarParameter, properties {defaultValue=1, group=DashSurge, parameterName=SubsurfaceOpacityMapMultiplier}}
  MaterialExpressionScalarParameter_2 {class=MaterialExpressionScalarParameter, properties {parameterName=flipbook_face_mouth_index}}
  MaterialExpressionScalarParameter_3 {class=MaterialExpressionScalarParameter, properties {parameterName=WartyBrine_MaskOn}}
  MaterialExpressionScalarParameter_4 {class=MaterialExpressionScalarParameter, properties {parameterName=WartyBrine_UseEyeMask}}
  MaterialExpressionSetMaterialAttributes_0 {
   class=MaterialExpressionSetMaterialAttributes
   inputs [
    MaterialAttributes {from=MaterialExpressionGetMaterialAttributes_0, output=MaterialAttributes}
    Shading Model {from=MaterialExpressionShadingModel_0, output=""}
    Opacity {from=MaterialExpressionLinearInterpolate_0, output=""}
   ]
   properties {attributeSetTypes [#5, #1]}
  }
  MaterialExpressionSetMaterialAttributes_2 {
   class=MaterialExpressionSetMaterialAttributes
   inputs [
    MaterialAttributes {from=MaterialExpressionGetMaterialAttributes_2, output=MaterialAttributes}
    World Position Offset {from=MaterialExpressionLinearInterpolate_2, output=""}
    Base Color {from="", output=""}
   ]
   properties {attributeSetTypes [#3, #4]}
  }
  MaterialExpressionSetMaterialAttributes_3 {
   class=MaterialExpressionSetMaterialAttributes
   inputs [
    MaterialAttributes {from=MaterialExpressionMaterialFunctionCall_0, output=Result}
    World Position Offset {from=MaterialExpressionMultiply_2, output=""}
   ]
   properties {attributeSetTypes [#3]}
  }
  MaterialExpressionShadingModel_0 {class=MaterialExpressionShadingModel, inputs [], properties {shadingModel=MSM_SubsurfaceProfile}}
  MaterialExpressionSphereMask_0 {
   class=MaterialExpressionSphereMask
   inputs [
    A {from=MaterialExpressionVertexColor_0, output=B}
    B {from=MaterialExpressionMax_0, output=""}
    Radius {from="", output=""}
    Hardness {from="", output=""}
   ]
   properties {attenuationRadius=0.03}
  }
  MaterialExpressionSphereMask_1 {
   class=MaterialExpressionSphereMask
   inputs [
    A {from=MaterialExpressionVertexColor_1, output=R}
    B {from=MaterialExpressionLinearInterpolate_1, output=""}
    Radius {from="", output=""}
    Hardness {from="", output=""}
   ]
   properties {attenuationRadius=0.2}
  }
  MaterialExpressionStaticSwitchParameter_0 {
   class=MaterialExpressionStaticSwitchParameter
   inputs [
    True {from=MaterialExpressionSetMaterialAttributes_2, output=""}
    False {from=MaterialExpressionSetMaterialAttributes_0, output=""}
   ]
   properties {group=WartyBrine, parameterName=IsHead}
  }
  MaterialExpressionStaticSwitchParameter_1 {
   class=MaterialExpressionStaticSwitchParameter
   inputs [
    True {from=MaterialExpressionSetMaterialAttributes_3, output=""}
    False {from=MaterialExpressionStaticSwitchParameter_0, output=""}
   ]
   properties {group=WartyBrine, parameterName=IsFaceAcc}
  }
  MaterialExpressionSwitch_0 {
   class=MaterialExpressionSwitch
   inputs [
    SwitchValue {from=MaterialExpressionScalarParameter_2, output=""}
    Default {from="", output=""}
    None {from=MaterialExpressionConstant_0, output=""}
    None {from=MaterialExpressionConstant_1, output=""}
    None {from=MaterialExpressionConstant_1, output=""}
    None {from=MaterialExpressionConstant_1, output=""}
    None {from=MaterialExpressionConstant_1, output=""}
    None {from=MaterialExpressionConstant_1, output=""}
    None {from=MaterialExpressionConstant_1, output=""}
    None {from=MaterialExpressionConstant_1, output=""}
    None {from=MaterialExpressionConstant_1, output=""}
    None {from=MaterialExpressionConstant_1, output=""}
    None {from=MaterialExpressionConstant_1, output=""}
    None {from=MaterialExpressionConstant2Vector_0, output=R}
    None {from=MaterialExpressionConstant2Vector_0, output=R}
    None {from=MaterialExpressionConstant_1, output=""}
    None {from=MaterialExpressionConstant_1, output=""}
   ]
   properties {inputs [
     {
      inputName=None
      input {
       expression=&self:MaterialExpressionConstant_0
       inputName=None
       mask=0
       maskA=0
       maskB=0
       maskG=0
       maskR=0
       outputIndex=0
      }
     }
     {
      inputName=None
      input {
       expression=&self:MaterialExpressionConstant_1
       inputName=None
       mask=0
       maskA=0
       maskB=0
       maskG=0
       maskR=0
       outputIndex=0
      }
     }
     {
      inputName=None
      input {
       expression=&self:MaterialExpressionConstant_1
       inputName=None
       mask=0
       maskA=0
       maskB=0
       maskG=0
       maskR=0
       outputIndex=0
      }
     }
     {
      inputName=None
      input {
       expression=&self:MaterialExpressionConstant_1
       inputName=None
       mask=0
       maskA=0
       maskB=0
       maskG=0
       maskR=0
       outputIndex=0
      }
     }
     {
      inputName=None
      input {
       expression=&self:MaterialExpressionConstant_1
       inputName=None
       mask=0
       maskA=0
       maskB=0
       maskG=0
       maskR=0
       outputIndex=0
      }
     }
     {
      inputName=None
      input {
       expression=&self:MaterialExpressionConstant_1
       inputName=None
       mask=0
       maskA=0
       maskB=0
       maskG=0
       maskR=0
       outputIndex=0
      }
     }
     {
      inputName=None
      input {
       expression=&self:MaterialExpressionConstant_1
       inputName=None
       mask=0
       maskA=0
       maskB=0
       maskG=0
       maskR=0
       outputIndex=0
      }
     }
     {
      inputName=None
      input {
       expression=&self:MaterialExpressionConstant_1
       inputName=None
       mask=0
       maskA=0
       maskB=0
       maskG=0
       maskR=0
       outputIndex=0
      }
     }
     {
      inputName=None
      input {
       expression=&self:MaterialExpressionConstant_1
       inputName=None
       mask=0
       maskA=0
       maskB=0
       maskG=0
       maskR=0
       outputIndex=0
      }
     }
     {
      inputName=None
      input {
       expression=&self:MaterialExpressionConstant_1
       inputName=None
       mask=0
       maskA=0
       maskB=0
       maskG=0
       maskR=0
       outputIndex=0
      }
     }
     {
      inputName=None
      input {
       expression=&self:MaterialExpressionConstant_1
       inputName=None
       mask=0
       maskA=0
       maskB=0
       maskG=0
       maskR=0
       outputIndex=0
      }
     }
     {
      inputName=None
      input {
       expression=&self:MaterialExpressionConstant2Vector_0
       inputName=None
       mask=1
       maskA=0
       maskB=0
       maskG=0
       maskR=1
       outputIndex=1
      }
     }
     {
      inputName=None
      input {
       expression=&self:MaterialExpressionConstant2Vector_0
       inputName=None
       mask=1
       maskA=0
       maskB=0
       maskG=1
       maskR=0
       outputIndex=2
      }
     }
     {
      inputName=None
      input {
       expression=&self:MaterialExpressionConstant_1
       inputName=None
       mask=0
       maskA=0
       maskB=0
       maskG=0
       maskR=0
       outputIndex=0
      }
     }
     {
      inputName=None
      input {
       expression=&self:MaterialExpressionConstant_1
       inputName=None
       mask=0
       maskA=0
       maskB=0
       maskG=0
       maskR=0
       outputIndex=0
      }
     }
    ]}
  }
  MaterialExpressionVertexColor_0 {class=MaterialExpressionVertexColor, inputs [], properties {}}
  MaterialExpressionVertexColor_1 {class=MaterialExpressionVertexColor, inputs [], properties {}}
  MaterialExpressionVertexColor_4 {class=MaterialExpressionVertexColor, inputs [], properties {}}
 }
 properties {
  bUseMaterialAttributes=true
  bUsedWithClothing=true
  bUsedWithMeshDeformer=true
  bUsedWithMorphTargets=true
  bUsedWithNanite=true
  bUsedWithNiagaraMeshParticles=true
  bUsedWithSkeletalMesh=true
  pixelDepthOffsetMode=PDOM_Legacy
  shadingModel=MSM_FromMaterialExpression
  usedShadingModels=Default Lit | Subsurface | Clear Coat | Subsurface Profile | Cloth
 }
}
