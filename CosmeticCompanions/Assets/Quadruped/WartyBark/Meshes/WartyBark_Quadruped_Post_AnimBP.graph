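# Post-process AnimBP for the WartyBark quadruped cosmetic companion, layering pose adjustments and tongue physics onto the incoming pose.
# On initialize it checks whether the owner is a FortCompanionMannequin; mannequins keep the input pose, others get AdditivePoseAnim weighted by CosmeticPoseAlpha.
# The result feeds a face PoseAsset, then chained AnimDynamics on tongue_02 and tongue_03, simulated bone-relative to spine_03 with inverted DynamicsAlpha as weight.
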AnimBlueprint /CosmeticCompanions/Assets/Quadruped/WartyBark/Meshes/WartyBark_Quadruped_Post_AnimBP
target_skeleton=&/WartyBark_Quadruped_Skeleton
graphs [
 AnimGraph {nodes [
   n1 = Get AdditivePoseAnim() -> (AdditivePoseAnim:Anim Sequence Object Reference)
   n2 = Blend Poses by bool(
    BlendPose_0=Use cached pose 'InputPoseCache'()
    BlendPose_1=Apply Additive(
     Base=Use cached pose 'InputPoseCache'()
     Additive=Sequence Player!(Sequence:Anim Sequence Base Object Reference=@n1) -> (Pose)
     Alpha=Get CosmeticPoseAlpha() -> (CosmeticPoseAlpha:float)
    ) {alphaBoolBlend {blendOption=Linear}}
    bActiveValue=Get isCompanionMannequin() -> (isCompanionMannequin:bool)
   )
   n3 = Get DynamicsAlpha() -> (DynamicsAlpha:float)
   n4 = Anim Dynamics - Bone: tongue_02!(
    ComponentPose=Local To Component(LocalPose=WartyBark_face_poses_PoseAsset!(SourcePose:Pose Link Structure=@n2) -> (Pose:Pose Link Structure) {
      is=title AnimGraphNode_PoseBlendNode
      blendOption=Linear
      poseAsset=&/WartyBark_face_poses_PoseAsset
     })
    Alpha=@n3
    GravityOverride:Vector="0,0,0"
    ExternalForce="0,0,0"
   ) -> (Pose) {
    angularDampingOverride=1
    angularSpringConstant=15000
    bAngularSpring=true
    bOverrideAngularDamping=true
    bUseGravityOverride=true
    bUsePlanarLimit=false
    gravityScale=0.35
    simulationSpace=BoneRelative
    alphaBoolBlend {blendOption=Linear}
    alphaScaleBiasClamp {bias=1, scale=-1}
    boundBone {boneName=tongue_02}
    componentAppliedLinearAccClamp=<3000 3000 3000>
    componentLinearAccScale=<1 1 1>
    componentLinearVelScale=<1 1 1>
    gravityOverride {z=-980}
    physicsBodyDefinitions [{
      boundBone {boneName=tongue_02}
      boxExtents=<20 30 30>
      constraintSetup {angularLimitsMax=<15 30 0>, angularLimitsMin=<-15 -10 0>, angularTarget=<1 0 0>}
      localJointOffset=<4 0 0>
     }]
    relativeSpaceBone {boneName=spine_03}
   }
   n5 = Anim Dynamics - Bone: tongue_03!(ComponentPose=@n4, Alpha=@n3, GravityOverride:Vector="0,0,0", ExternalForce="0,0,0") -> (Pose) {
    angularDampingOverride=1
    angularSpringConstant=15000
    bAngularSpring=true
    bOverrideAngularDamping=true
    bUseGravityOverride=true
    bUsePlanarLimit=false
    gravityScale=0.35
    simulationSpace=BoneRelative
    alphaBoolBlend {blendOption=Linear}
    alphaScaleBiasClamp {bias=1, scale=-1}
    boundBone {boneName=tongue_03}
    componentAppliedLinearAccClamp=<3000 3000 3000>
    componentLinearAccScale=<1 1 1>
    componentLinearVelScale=<1 1 1>
    gravityOverride {z=-980}
    physicsBodyDefinitions [{
      boundBone {boneName=tongue_03}
      boxExtents=<20 30 30>
      constraintSetup {angularLimitsMax=<30 45 5>, angularLimitsMin=<-30 -5 -5>, angularTarget=<1 0 0>}
      localJointOffset=<4 0 0>
     }]
    relativeSpaceBone {boneName=spine_03}
   }
   Output Pose(Result=Component To Local(ComponentPose=@n5))
   Save cached pose 'InputPoseCache'(Pose=Input Pose()) {cacheName=InputPoseCache}
  ]}
 EventGraph {nodes [
   Event BlueprintInitializeAnimation()
   > Set isCompanionMannequin(isCompanionMannequin:bool=IsA!(Object:Object Reference=GetOwningActor(), ObjectClass:Object Class Reference=/Script/CosmeticCompanionsRuntime.FortCompanionMannequin) -> (ReturnValue:bool)) -> (Output_Get:bool)
  ]}
]
variables [isCompanionMannequin {type=boolean}]
